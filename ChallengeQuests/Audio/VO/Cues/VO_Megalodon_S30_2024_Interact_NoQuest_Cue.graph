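# Sound cue for the Megalodon S30 2024 NPC's voice line, played, per its name, on an interaction with no quest available.
# A random node picks one of three equally weighted wave variants.
# It uses shared close-range NPC voice attenuation and an enemy critical VO sound class with boosted volume.
# A concurrency override allows one instance per owner and blocks new plays while one is active.
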SoundCue /ChallengeQuests/Audio/VO/Cues/VO_Megalodon_S30_2024_Interact_NoQuest_Cue
root=Random_0
nodes [
 {id=Random_0, children [WavePlayer_0,WavePlayer_1,WavePlayer_2], properties {weights [1,1,1]}}
 {id=WavePlayer_0, properties {soundWaveAssetPtr=/ChallengeQuests/Audio/VO/Waves/VO_Megalodon_S30_2024_Interact_NoQuest_010}}
 {id=WavePlayer_1, properties {soundWaveAssetPtr=/ChallengeQuests/Audio/VO/Waves/VO_Megalodon_S30_2024_Interact_NoQuest_020}}
 {id=WavePlayer_2, properties {soundWaveAssetPtr=/ChallengeQuests/Audio/VO/Waves/VO_Megalodon_S30_2024_Interact_NoQuest_030}}
]
properties {
 attenuationSettings=/Game/Sounds/Fort_Human_VO_BR/NPCs/Shared/NPC_VO_Close_Att
 bOverrideConcurrency=true
 maxDistance=8500
 soundClassObject=&Fort_VO_Enemy_Critical
 volumeMultiplier=2
 concurrencyOverrides {bLimitToOwner=true, maxCount=1, resolutionRule=PreventNew}
}
